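# Niagara effect for the Endear emote that throws out a small burst of little pink mesh hearts.
# A single LittleHearts emitter burst-spawns short-lived mesh particles with randomized HDR tint, size and slightly tilted orientation, launched mostly along +Y.
# Drag slows the hearts while a randomized upward acceleration lifts them.
# A scale curve grows each heart, swells it briefly, then shrinks it to nothing at end of life.
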NiagaraSystem /BRCosmetics/Effects/Fort_Effects/Emotes/Endear/NS_Emote_Endear_Hearts
dependencies {usedDataInterfaces [&NiagaraDataInterfaceCurve], usedDynamicInputs [], usedModules [&SystemState]}
emitters {LittleHearts {
  emitterUpdateScript [
   &EmitterState(Life Cycle Mode=::NewEnumerator1, Loop Behavior=::NewEnumerator1, Loop Duration=1, UseLoopDelay=true)
   &SpawnBurst_Instantaneous(Spawn Count=5, Use Spawn Probability=false, UseLoopCountLimit=false)
  ]
  particleSpawnScript [
   &InitializeParticle(
    Color=&RandomRangeLinearColor(Minimum=<rgba 3 0.3344 0.5311 1>, Maximum=<rgba 3 0.0949 0.3092 1>)
    Lifetime=&RandomRangeFloat(Minimum=0.6, Maximum=1.25)
    Mesh Scale=&VectorFromFloat(Value=&RandomRangeFloat(Minimum=0.1, Maximum=0.16))
    Mesh Scale Mode=::NewEnumerator4
    UsePositionOffset=false
    Write Material Random=true
   )
   &AddVelocity(Velocity=&RandomRangeVector(Minimum=<-60 30 -20>, Maximum=<60 145 60>))
   &InitialMeshOrientation(
    Mesh Orientation Mode=::NewEnumerator6
    Orientation Quaternion:Quat4f=&MakeQuatFromAxisAngle(Axis=&RandomRangeVector(Minimum=<-0.05 -0.03 0>, Maximum=<0.05 0.03 0>), AngleInDegrees=90)
    Rotation=<0 0 0.25>
    Use Rotation Vector=true
   )
  ]
  particleUpdateScript [
   &ParticleState()
   &ScaleMeshSize(Scale Factor=&VectorFromFloat(Value=&FloatFromCurve(FloatCurve=DataInterface() {propertyValues {Curve {keys [
          {arriveTangent=3.6183, interpMode=RCIM_Cubic, time=0, value=0.2}
          {arriveTangent=1.1429, interpMode=RCIM_Cubic, leaveTangent=1.1429, time=0.4, value=1}
          {arriveTangent=0.75, interpMode=RCIM_Cubic, leaveTangent=0.75, time=0.7, value=1}
          {arriveTangent=-3.3333, interpMode=RCIM_Cubic, leaveTangent=-3.3333, time=0.8, value=1.3}
          {interpMode=RCIM_Cubic, leaveTangent=-33.019, time=1, value=0}
         ]}}})))
   &Drag(Drag=4, Use Linear Drag=true, Use Rotational Drag=true)
   &AccelerationForce(Acceleration=&RandomRangeVector(Minimum=<0 0 30>, Maximum=<0 0 60>))
   &SolveForcesAndVelocity(Clamp Velocity=false, Limit Acceleration=false)
  ]
  rendererClasses [&NiagaraMeshRendererProperties]
  renderers [{rendererClass=&NiagaraMeshRendererProperties, rendererIndex=0}]
 }}
summary {
 systemName=NS_Emote_Endear_Hearts
 emitters [{emitterName=LittleHearts, rendererClasses [&NiagaraMeshRendererProperties]}]
 userVariables []
}
